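# Unlit additive Niagara particle material that makes sprites act like decals, painting animated glow onto stencil-marked surfaces, probably vehicles.
# A panning, distorted texture, world-aligned noise and dynamic dissolve and distortion parameters shape the pattern.
# Scene-normal Fresnel, a gradient mask and particle color build the emissive.
# Opacity uses a depth-checked custom stencil on deferred, switching to a depth and particle-distance mask on forward and mobile.
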Material /VehicleCosmetics/FX/Niagara/Prototypes/Limes/Bright/Material/M_ParticleDecal_Sprites
graph {
 properties {
  bUsedWithNanite=true
  bUsedWithNiagaraMeshParticles=true
  bUsedWithNiagaraSprites=true
  blendMode=BLEND_Additive
  pixelDepthOffsetMode=PDOM_Legacy
  shadingModel=MSM_Unlit
  translucencyPass=MTP_BeforeDOF
  usedShadingModels=Unlit
 }
 nodes [
  n1 = Multiply(A=VectorParameter() {parameterName=World_Noise_Panner, defaultValue=<rgba 1 1 1 0>}.RGB) {constB=100}
  n2 = Time()
  n3 = MaterialFunctionCall(
   TextureObject=TextureObject() {samplerType=SAMPLERTYPE_LinearGrayscale, texture=&T_SimplePerlin_Greyscale}
   TextureSize=VectorParameter() {parameterName=World_Noise_Size, defaultValue=<rgba 300 300 300 0>}.RGB
   WorldPosition=Add(A=WorldPosition().XYZ, B=Multiply(A=@n1, B=@n2))
   World Space Normal=SceneTexture() {sceneTextureId=PPI_WorldNormal}.Color
  ) {materialFunction=&WorldAlignedTexture}
  n4 = ScalarParameter() {defaultValue=2, parameterName=World Noise - Exp}
  NamedRerouteDeclaration(None=Clamp(None=Power(Base=@n3.XYZ Texture, Exp=@n4), Min=ScalarParameter() {parameterName=World Noise - Bias})) {name=World Aligned Noise, nodeColor=<rgba 0.8047 1 0 1>}
  n5 = VectorParameter() {group=Texture, parameterName=Texture -Tiling, channelNames {a=Speed V, b=Speed U, g=Tiling V, r=Tiling U}, defaultValue=<rgba 1 1 0 0>}
  n6 = AppendVector(A=@n5.Tiling U, B=@n5.Tiling U)
  n7 = ScalarParameter() {defaultValue=1, group=Global, parameterName=Global Tiling}
  n8 = Panner(
   Coordinate=Multiply(A=TextureCoordinate(), B=Multiply(A=@n6, B=@n7) {constB=0.15})
   Time=Time()
   Speed=Multiply(A=AppendVector(A=@n5.Speed U, B=@n5.Speed U), B=ScalarParameter() {defaultValue=1, group=Global, parameterName=Global Speed}) {constB=0.15}
  )
  n9 = ComponentMask(None=Multiply(A=NamedRerouteUsage(), B=VectorParameter() {parameterName=DP_Panner_Direction, defaultValue=<rgba 1 0 0 1>}.RGB)) {g=true, r=true}
  n10 = Multiply(A=ScalarParameter() {defaultValue=0.1, group=Distortion, parameterName=Distortion - Amount}, B=NamedRerouteUsage())
  n11 = TextureCoordinate()
  n12 = VectorParameter() {group=Distortion, parameterName=Distortion -Tiling, channelNames {a=Speed V, b=Speed U, g=Tiling V, r=Tiling U}, defaultValue=<rgba 3 3 0 0>}
  n13 = Multiply(A=AppendVector(A=@n12.Tiling U, B=@n12.Tiling U), B=ScalarParameter() {defaultValue=1, group=Global, parameterName=Global Tiling}) {constB=0.15}
  n14 = AppendVector(A=@n12.Speed U, B=@n12.Speed U)
  n15 = ScalarParameter() {defaultValue=1, group=Global, parameterName=Global Speed}
  n16 = TextureSampleParameter2D(UVs=Panner(Coordinate=Multiply(A=@n11, B=@n13), Time=Time(), Speed=Multiply(A=@n14, B=@n15) {constB=0.15})) {
   group=Distortion
   parameterName=Distortion - Texture
   samplerType=SAMPLERTYPE_LinearGrayscale
   texture=/Game/Effects/Fort_Effects/Emotes/UltraLight/Texture/T_Tiling_Noise_01_GreyScale
  }
  n17 = NamedRerouteUsage()
  n18 = ScalarParameter() {defaultValue=2, group=Distortion, parameterName=Distortion - World Aligned Amount}
  n19 = ComponentMask(None=Add(A=@n16.RGB, B=Multiply(A=@n17, B=@n18))) {r=true}
  n20 = TextureSampleParameter2D(UVs=Add(A=Add(A=@n8, B=@n9), B=Multiply(A=@n19, B=@n10) {constB=0.75})) {group=Texture, parameterName=Texture - Texture, texture=/DealershipSharedLibrary/FX/Textures/TexturePackage/T_VCVFX_Shapes_01_Pack_RGB}
  n21 = Power(Base=@n20.RGB, Exp=ScalarParameter() {defaultValue=1, group=Texture, parameterName=Texture - Exp})
  n22 = NamedRerouteUsage()
  NamedRerouteDeclaration(None=SmoothStep(Max=NamedRerouteUsage(), Value=Multiply(A=@n21, B=@n22)) {constValue=1}) {name=Texture, nodeColor=<rgba 0 0.5547 1 1>}
  n23 = DynamicParameter() {paramNames [Dissolve,Disortion,Panner,Param4]}
  NamedRerouteDeclaration(None=@n23.Dissolve) {name=DP_Dissolve, nodeColor=<rgba 1 0.0082 0.0045 1>}
  n24 = ParticleColor()
  n25 = Clamp(
   None=Fresnel(
    ExponentIn=ScalarParameter() {defaultValue=3, group=Global, parameterName=Global - Fresnel Exp}
    BaseReflectFractionIn=ScalarParameter() {group=Global, parameterName=Global - Fresnel Reflection}
    Normal=SceneTexture() {sceneTextureId=PPI_WorldNormal}.Color
   ) {baseReflectFraction=0, exponent=2}
   Min=ScalarParameter() {defaultValue=0.05, group=Global, parameterName=Global - Emissive Min}
  )
  n26 = ScalarParameter() {defaultValue=40, group=Global, parameterName=Global - Emissive Amount}
  n27 = TextureSampleParameter2D() {group=Opacity, parameterName=Opacity - Mask Texture, texture=/DealershipSharedLibrary/FX/Textures/Gradient_Masks/T_VCVFX_SphereGrad_A}
  n28 = Multiply(A=Multiply(A=NamedRerouteUsage(), B=Multiply(A=@n25, B=@n26) {constB=2}), B=@n27.RGB)
  NamedRerouteDeclaration(None=MaterialFunctionCall(Emissive=Multiply(A=@n24.RGB, B=@n28)) {materialFunction=&MF_FX_EmissiveAdjustments}.Result) {name=Emissive, nodeColor=<rgba 1 0 0.7969 1>}
  NamedRerouteDeclaration(None=@n23.Disortion) {name=DP_Distortion, nodeColor=<rgba 1 0.0082 0.0045 1>}
  NamedRerouteDeclaration(None=@n23.Panner) {name=DP_Panner, nodeColor=<rgba 1 0.0082 0.0045 1>}
  n29 = ScreenPosition()
  n30 = ComponentMask(None=SceneTexture(UVs=@n29.ViewportUV) {sceneTextureId=PPI_CustomStencil}.Color) {r=true}
  n31 = ComponentMask(None=SceneTexture(UVs=@n29.ViewportUV) {sceneTextureId=PPI_CustomStencil}.Color) {r=true}
  n32 = Subtract(A=ComponentMask(None=SceneTexture() {sceneTextureId=PPI_CustomDepth}.Color) {r=true}, B=SceneTexture() {sceneTextureId=PPI_SceneDepth}.Color)
  n33 = Constant()
  n34 = If(
   A=Subtract(A=Max(A=@n30, B=@n31), B=Saturate(None=@n32))
   B=ScalarParameter() {defaultValue=16, group=Opacity, parameterName=Opacity - Stencil Mask Value}
   "A > B"=@n33
   "A == B"=Constant() {r=1}
   "A < B"=@n33
  )
  n35 = SceneDepth(UVs=ScreenPosition().ViewportUV)
  n36 = Subtract(A=@n35.R, B=PixelDepth().R)
  n37 = Constant() {r=500}
  n38 = Saturate(None=OneMinus(None=Divide(A=@n36, B=@n37)))
  n39 = Distance(
   A=MaterialFunctionCall(SceneDepth=@n35.R) {materialFunction=&WorldPositionBehindTranslucency}.Result
   B=Add(
    A=ParticlePositionWS()
    B=VectorParameter() {group=Opacity, parameterName=Opacity - Mobile Mask Position Offset, defaultValue=<rgba 0 0 175 1>}.RGB
   )
  )
  n40 = Constant() {r=350}
  n41 = Saturate(None=OneMinus(None=Divide(A=@n39, B=@n40)))
  n42 = Multiply(
   A=Power(Base=Multiply(A=@n38, B=@n41), Exp=ScalarParameter() {defaultValue=3, group=Opacity, parameterName=Opacity - Mobile Mask Pow})
   B=ScalarParameter() {defaultValue=11.4011, group=Opacity, parameterName=Opacity - Mobile Mask Amount}
  )
  NamedRerouteDeclaration(None=Multiply(A=ParticleColor().A, B=ShadingPathSwitch(Default=@n34, Deferred=@n34, Forward=@n42, Mobile=@n42))) {name=Opacity, nodeColor=<rgba 0 1 0.8828 1>}
  NamedRerouteUsage()
  NamedRerouteUsage()
 ]
}
